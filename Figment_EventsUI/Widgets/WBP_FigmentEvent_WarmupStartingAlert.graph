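# Figment event warmup-starting alert widget, showing the static message "Event Protection Starting Soon".
# An Overlay holds a centered, padded HorizontalBox with a single uppercase CommonTextBlock in the Burbank font.
# The event graph has only unwired stub events, so the widget has no logic of its own.
# The text block is named Text_ProtectionEndingSoon, although its text says protection is starting.
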WidgetBlueprint /Figment_EventsUI/Widgets/WBP_FigmentEvent_WarmupStartingAlert
graphs [EventGraph {nodes [
   Event PreConstruct()
   Event Construct()
   Event Tick()
  ]}]
variables []
widgets {
 info {inheritedWidgetCount=0, parentClass=&UserWidget, rootWidgetClass=&Overlay, widgetCount=3}
 widgets [
  Overlay_49 {widgetClassPath=&Overlay, widget=Overlay_49}
  HorizontalBox_0 {
   widgetClassPath=&HorizontalBox
   parent=Overlay_49
   widget=HorizontalBox_0
   slot {class=OverlaySlot, properties {horizontalAlignment=HAlign_Center, verticalAlignment=VAlign_Center, padding=<ltrb 20 10 24 10>}}
  }
  Text_ProtectionEndingSoon {
   widgetClassPath=&CommonTextBlock
   bIsVariable=true
   parent=HorizontalBox_0
   widget=Text_ProtectionEndingSoon
   properties {
    text=Event Protection Starting Soon
    textTransformPolicy=ToUpper
    font {
     bForceMonospaced=false
     bMaterialIsStencil=false
     fontMaterial=null
     fontObject=&Burbank
     letterSpacing=0
     monospacedWidth=1
     size=19.5
     skewAmount=0
     typefaceFontName=BurbankRegular-Black
     outlineSettings {
      bApplyOutlineToDropShadows=false
      bMiteredCorners=false
      bSeparateFillAlpha=false
      outlineMaterial=null
      outlineSize=0
      outlineColor=<rgba 0 0 0 1>
     }
    }
    shadowOffset=<0 0>
   }
   slot {class=HorizontalBoxSlot, properties {verticalAlignment=VAlign_Center, padding=<ltrb 10 0 0 0>}}
  }
 ]
}
